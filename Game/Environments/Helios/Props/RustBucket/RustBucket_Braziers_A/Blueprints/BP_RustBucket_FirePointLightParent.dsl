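; Point light Blueprint for RustBucket brazier props, named as a parent class for fire lights, with an optional flame flicker.
; Its construction script runs the parent's script, then checks the isFlame flag to set the point light's light function material to the flicker function or clear it.
; BeginPlay, overlap and Tick events are empty, so this graph adds no runtime logic.
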(event EventBeginPlay)

(event Collision|EventActorBeginOverlap
  (params (OtherActor :type Actor%20Object%20Reference)))

(event EventTick
  (params (DeltaSeconds :type Float%20%28single-precision%29)))

(fn ConstructionScript
  (|CallParentFunction)
  (if (Variables|Building|TimeOfDayLights|GetisFlame)
    (Rendering|Components|Light|SetLightFunctionMaterial (Variables|Default|GetPointLight) (Variables|Building|TimeOfDayLights|GetFlameFlickerFunction))
    (else
      (Rendering|Components|Light|SetLightFunctionMaterial (Variables|Default|GetPointLight) 0))))
